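# Animation Blueprint for the male medium StreetBasketball body, adding secondary motion to shorts and necklace over a pose copied from the owner mesh.
# Anim Dynamics chains swing the shorts, calf-driven Flap nodes adjust the back hems, and the first necklace bone is simulated before caching.
# bCanPlayCustomAnimations switches the second necklace bone between dynamics and a layered ChainRotation animation.
# The third bone's dynamics and additive right-shorts offsets finish the pose.
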AnimBlueprint /Game/Characters/Player/Male/Medium/Bodies/M_MED_StreetBasketball/Meshes/M_MED_StreetBasketball_AnimBP
target_skeleton=&/M_MED_StreetBasketball_Skeleton
graphs [
 AnimGraph {nodes [
   n1 = Anim Dynamics - Bone: dyn_bball_necklace_02!(ComponentPose=Local To Component(LocalPose=Use cached pose 'cache_01'()), ExternalForce="0,0,0") -> (Pose) {
    alpha=0.6
    angularDampingOverride=0.99
    angularSpringConstant=1500
    bAngularSpring=true
    bGravityOverrideInSimSpace=true
    bOverrideAngularDamping=true
    numSolverIterationsPostUpdate=4
    numSolverIterationsPreUpdate=16
    simulationSpace=RootRelative
    alphaBoolBlend {blendOption=Linear}
    boundBone {boneName=dyn_bball_necklace_02}
    physicsBodyDefinitions [{
      sphereCollisionRadius=0
      boundBone {boneName=dyn_bball_necklace_02}
      boxExtents=<15 5 10>
      constraintSetup {angularTargetAxis=AxisY, angularLimitsMax=<0 60 60>, angularLimitsMin=<0 -60 0>, angularTarget=<0 1 0>}
      localJointOffset=<2 0 0>
     }]
   }
   n2 = Component To Local(ComponentPose=@n1)
   n3 = Layered blend per bone(
    BasePose=Use cached pose 'cache_01'()
    BlendPoses_0=Sequence Player 'M_MED_StreetBasketball_ChainRotation'() {sequence=/Game/Characters/Player/Male/Medium/Bodies/M_MED_StreetBasketball/Animation/M_MED_StreetBasketball_ChainRotation}
   ) {layerSetup [{branchFilters [{blendDepth=1, boneName=dyn_bball_necklace_02}]}]}
   n4 = Get bCanPlayCustomAnimations() -> (bCanPlayCustomAnimations:bool)
   n5 = Anim Dynamics - Bone: dyn_bball_necklace_03!(
    ComponentPose=Local To Component(LocalPose=Blend Poses by bool(BlendPose_0=@n3, BlendPose_1=@n2, bActiveValue=@n4) {blendType=Linear})
    ExternalForce="0,0,0"
   ) -> (Pose) {
    alpha=0.6
    angularDampingOverride=0.99
    angularSpringConstant=1500
    bAngularSpring=true
    bGravityOverrideInSimSpace=true
    bOverrideAngularDamping=true
    numSolverIterationsPostUpdate=4
    numSolverIterationsPreUpdate=16
    simulationSpace=RootRelative
    alphaBoolBlend {blendOption=Linear}
    boundBone {boneName=dyn_bball_necklace_03}
    physicsBodyDefinitions [{
      sphereCollisionRadius=0
      boundBone {boneName=dyn_bball_necklace_03}
      boxExtents=<10 5 10>
      constraintSetup {angularTargetAxis=AxisY, angularLimitsMax=<0 60 60>, angularLimitsMin=<0 -60 0>, angularTarget=<0 1 0>}
      localJointOffset=<2 0 0>
     }]
   }
   n6 = "Transform (Modify) Bone - Bone: dyn_shorts_bk_r"!(ComponentPose=@n5, Alpha=1, Translation="-0.181,0.202,-2.594", Rotation="4.916,-0.869,-10.035") -> (Pose) {
    rotationMode=BMM_Additive
    rotationSpace=BCS_BoneSpace
    translationMode=BMM_Additive
    translationSpace=BCS_BoneSpace
    alphaBoolBlend {blendOption=Linear}
    alphaScaleBiasClamp {bInterpResult=true}
    boneToModify {boneName=dyn_shorts_bk_r}
    rotation=<rot 4.9159 -0.869 -10.035>
    translation=<-0.181 0.202 -2.594>
   }
   n7 = "Transform (Modify) Bone - Bone: dyn_shorts_out_02_r"!(ComponentPose=@n6, Alpha=1, Translation="-0.425,0.889,-5.167") -> (Pose) {
    rotationMode=BMM_Additive
    rotationSpace=BCS_BoneSpace
    scaleMode=BMM_Additive
    scaleSpace=BCS_BoneSpace
    translationMode=BMM_Additive
    translationSpace=BCS_BoneSpace
    alphaBoolBlend {blendOption=Linear}
    alphaScaleBiasClamp {bInterpResult=true}
    boneToModify {boneName=dyn_shorts_out_02_r}
    translation=<-0.425 0.889 -5.167>
   }
   Output Pose(Result=Component To Local(ComponentPose=@n7)) {layerGroup=None}
   n8 = Get MeshToCopy() -> (MeshToCopy:Skeletal Mesh Component Object Reference)
   n9 = "Anim Dynamics - Chain: dyn_shorts_in_l -> dyn_shorts_in_02_l"!(
    ComponentPose=Local To Component(LocalPose=Copy Pose From Mesh(SourceMeshComponent=@n8))
    GravityOverride:Vector="0,0,-980"
    ExternalForce="0,0,0"
   ) -> (Pose) {
    angularDampingOverride=0.999
    angularSpringConstant=100
    bAngularSpring=true
    bChain=true
    bGravityOverrideInSimSpace=true
    bOverrideAngularDamping=true
    bUseGravityOverride=true
    numSolverIterationsPostUpdate=4
    numSolverIterationsPreUpdate=16
    simulationSpace=RootRelative
    alphaBoolBlend {blendOption=Linear}
    boundBone {boneName=dyn_shorts_in_l}
    chainEnd {boneName=dyn_shorts_in_02_l}
    gravityOverride {z=-980}
    physicsBodyDefinitions [
     {
      sphereCollisionRadius=0
      boundBone {boneName=dyn_shorts_in_l}
      boxExtents=<25 30 12>
      constraintSetup {angularTargetAxis=AxisZ, twistAxis=AxisZ, angularLimitsMax=<10 10 5>, angularLimitsMin=<-10 -10 -5>, angularTarget=<0 0 1>}
      localJointOffset=<0 0 -3.5>
     }
     {
      boundBone {boneName=dyn_shorts_in_02_l}
      boxExtents=<10 10 10>
      constraintSetup {angularLimitsMax=<0 0 0>, angularLimitsMin=<0 0 0>, angularTarget=<0 0 0>}
      localJointOffset=<0 0 0>
     }
    ]
   }
   n10 = "Anim Dynamics - Chain: dyn_shorts_out_l -> dyn_shorts_out_02_l"!(ComponentPose=@n9, GravityOverride:Vector="0,0,-980", ExternalForce="0,0,0") -> (Pose) {
    angularDampingOverride=0.999
    angularSpringConstant=100
    bAngularSpring=true
    bChain=true
    bGravityOverrideInSimSpace=true
    bOverrideAngularDamping=true
    bUseGravityOverride=true
    numSolverIterationsPostUpdate=4
    numSolverIterationsPreUpdate=16
    simulationSpace=RootRelative
    alphaBoolBlend {blendOption=Linear}
    boundBone {boneName=dyn_shorts_out_l}
    chainEnd {boneName=dyn_shorts_out_02_l}
    gravityOverride {z=-980}
    physicsBodyDefinitions [
     {
      sphereCollisionRadius=0
      boundBone {boneName=dyn_shorts_out_l}
      boxExtents=<50 50 30>
      constraintSetup {angularTargetAxis=AxisZ, twistAxis=AxisZ, angularLimitsMax=<10 10 5>, angularLimitsMin=<-10 -10 -5>, angularTarget=<0 0 1>}
      localJointOffset=<0 0 -2>
     }
     {
      boundBone {boneName=dyn_shorts_out_02_l}
      boxExtents=<10 10 10>
      constraintSetup {angularLimitsMax=<0 0 0>, angularLimitsMin=<0 0 0>, angularTarget=<0 0 0>}
      localJointOffset=<0 0 0>
     }
    ]
   }
   n11 = "Anim Dynamics - Chain: dyn_shorts_fr_l -> dyn_shorts_fr_02_l"!(ComponentPose=@n10, GravityOverride:Vector="0,0,-980", ExternalForce="0,0,0") -> (Pose) {
    angularDampingOverride=0.999
    angularSpringConstant=300
    bAngularSpring=true
    bChain=true
    bGravityOverrideInSimSpace=true
    bOverrideAngularDamping=true
    bUseGravityOverride=true
    numSolverIterationsPostUpdate=4
    numSolverIterationsPreUpdate=16
    simulationSpace=RootRelative
    alphaBoolBlend {blendOption=Linear}
    boundBone {boneName=dyn_shorts_fr_l}
    chainEnd {boneName=dyn_shorts_fr_02_l}
    gravityOverride {z=-980}
    physicsBodyDefinitions [
     {
      sphereCollisionRadius=0
      boundBone {boneName=dyn_shorts_fr_l}
      boxExtents=<18 25 8>
      constraintSetup {angularTargetAxis=AxisZ, twistAxis=AxisZ, angularLimitsMax=<10 0 5>, angularLimitsMin=<-10 -20 -5>, angularTarget=<0 0 1>}
      localJointOffset=<0 0 -5>
     }
     {
      boundBone {boneName=dyn_shorts_fr_02_l}
      boxExtents=<10 10 10>
      constraintSetup {angularLimitsMax=<0 0 0>, angularLimitsMin=<0 0 0>, angularTarget=<0 0 0>}
      localJointOffset=<0 0 0>
     }
    ]
   }
   n12 = "Anim Dynamics - Chain: dyn_shorts_bk_l -> dyn_shorts_bk_02_l"!(ComponentPose=@n11, GravityOverride:Vector="0,0,-980", ExternalForce="0,0,0") -> (Pose) {
    angularDampingOverride=0.999
    angularSpringConstant=300
    bAngularSpring=true
    bChain=true
    bGravityOverrideInSimSpace=true
    bOverrideAngularDamping=true
    bUseGravityOverride=true
    numSolverIterationsPostUpdate=4
    numSolverIterationsPreUpdate=16
    simulationSpace=RootRelative
    alphaBoolBlend {blendOption=Linear}
    boundBone {boneName=dyn_shorts_bk_l}
    chainEnd {boneName=dyn_shorts_bk_02_l}
    gravityOverride {z=-980}
    physicsBodyDefinitions [
     {
      sphereCollisionRadius=0
      boundBone {boneName=dyn_shorts_bk_l}
      boxExtents=<40 38 15>
      constraintSetup {angularTargetAxis=AxisZ, twistAxis=AxisZ, angularLimitsMax=<10 0 5>, angularLimitsMin=<-10 -20 -5>, angularTarget=<0 0 1>}
      localJointOffset=<0 0 -3.5>
     }
     {
      boundBone {boneName=dyn_shorts_bk_02_l}
      boxExtents=<10 10 10>
      constraintSetup {angularLimitsMax=<0 0 0>, angularLimitsMin=<0 0 0>, angularTarget=<0 0 0>}
      localJointOffset=<0 0 0>
     }
    ]
   }
   n13 = "Anim Dynamics - Chain: dyn_shorts_in_r -> dyn_shorts_in_02_r"!(ComponentPose=@n12, GravityOverride:Vector="0,0,-980", ExternalForce="0,0,0") -> (Pose) {
    angularDampingOverride=0.999
    angularSpringConstant=100
    bAngularSpring=true
    bChain=true
    bGravityOverrideInSimSpace=true
    bOverrideAngularDamping=true
    bUseGravityOverride=true
    numSolverIterationsPostUpdate=4
    numSolverIterationsPreUpdate=16
    simulationSpace=RootRelative
    alphaBoolBlend {blendOption=Linear}
    boundBone {boneName=dyn_shorts_in_r}
    chainEnd {boneName=dyn_shorts_in_02_r}
    gravityOverride {z=-980}
    physicsBodyDefinitions [
     {
      sphereCollisionRadius=0
      boundBone {boneName=dyn_shorts_in_r}
      boxExtents=<25 29 12>
      constraintSetup {angularTargetAxis=AxisZ, twistAxis=AxisZ, angularLimitsMax=<10 10 5>, angularLimitsMin=<-10 -10 -5>, angularTarget=<0 0 1>}
      localJointOffset=<0 0 3.5>
     }
     {
      boundBone {boneName=dyn_shorts_in_02_r}
      boxExtents=<10 10 10>
      constraintSetup {angularLimitsMax=<0 0 0>, angularLimitsMin=<0 0 0>, angularTarget=<0 0 0>}
      localJointOffset=<0 0 0>
     }
    ]
   }
   n14 = "Anim Dynamics - Chain: dyn_shorts_out_r -> dyn_shorts_out_02_r"!(ComponentPose=@n13, GravityOverride:Vector="0,0,-980", ExternalForce="0,0,0") -> (Pose) {
    angularDampingOverride=0.999
    angularSpringConstant=100
    bAngularSpring=true
    bChain=true
    bGravityOverrideInSimSpace=true
    bOverrideAngularDamping=true
    bUseGravityOverride=true
    numSolverIterationsPostUpdate=4
    numSolverIterationsPreUpdate=16
    simulationSpace=RootRelative
    alphaBoolBlend {blendOption=Linear}
    boundBone {boneName=dyn_shorts_out_r}
    chainEnd {boneName=dyn_shorts_out_02_r}
    gravityOverride {z=-980}
    physicsBodyDefinitions [
     {
      sphereCollisionRadius=0
      boundBone {boneName=dyn_shorts_out_r}
      boxExtents=<50 50 30>
      constraintSetup {angularTargetAxis=AxisZ, twistAxis=AxisZ, angularLimitsMax=<10 10 5>, angularLimitsMin=<-10 -10 -5>, angularTarget=<0 0 1>}
      localJointOffset=<0 0 2>
     }
     {
      boundBone {boneName=dyn_shorts_out_02_r}
      boxExtents=<10 10 10>
      constraintSetup {angularLimitsMax=<0 0 0>, angularLimitsMin=<0 0 0>, angularTarget=<0 0 0>}
      localJointOffset=<0 0 0>
     }
    ]
   }
   n15 = "Anim Dynamics - Chain: dyn_shorts_fr_r -> dyn_shorts_fr_02_r"!(ComponentPose=@n14, GravityOverride:Vector="0,0,-980", ExternalForce="0,0,0") -> (Pose) {
    angularDampingOverride=0.999
    angularSpringConstant=300
    bAngularSpring=true
    bChain=true
    bGravityOverrideInSimSpace=true
    bOverrideAngularDamping=true
    bUseGravityOverride=true
    numSolverIterationsPostUpdate=4
    numSolverIterationsPreUpdate=16
    simulationSpace=RootRelative
    alphaBoolBlend {blendOption=Linear}
    boundBone {boneName=dyn_shorts_fr_r}
    chainEnd {boneName=dyn_shorts_fr_02_r}
    gravityOverride {z=-980}
    physicsBodyDefinitions [
     {
      sphereCollisionRadius=0
      boundBone {boneName=dyn_shorts_fr_r}
      boxExtents=<26 22 8>
      constraintSetup {angularTargetAxis=AxisZ, twistAxis=AxisZ, angularLimitsMax=<10 0 5>, angularLimitsMin=<-10 -20 -5>, angularTarget=<0 0 1>}
      localJointOffset=<0 0 5>
     }
     {
      boundBone {boneName=dyn_shorts_fr_02_r}
      boxExtents=<10 10 10>
      constraintSetup {angularLimitsMax=<0 0 0>, angularLimitsMin=<0 0 0>, angularTarget=<0 0 0>}
      localJointOffset=<0 0 0>
     }
    ]
   }
   n16 = "Anim Dynamics - Chain: dyn_shorts_bk_r -> dyn_shorts_bk_02_r"!(ComponentPose=@n15, GravityOverride:Vector="0,0,-980", ExternalForce="0,0,0") -> (Pose) {
    angularDampingOverride=0.999
    angularSpringConstant=300
    bAngularSpring=true
    bChain=true
    bGravityOverrideInSimSpace=true
    bOverrideAngularDamping=true
    bUseGravityOverride=true
    numSolverIterationsPostUpdate=4
    numSolverIterationsPreUpdate=16
    simulationSpace=RootRelative
    alphaBoolBlend {blendOption=Linear}
    boundBone {boneName=dyn_shorts_bk_r}
    chainEnd {boneName=dyn_shorts_bk_02_r}
    gravityOverride {z=-980}
    physicsBodyDefinitions [
     {
      sphereCollisionRadius=0
      boundBone {boneName=dyn_shorts_bk_r}
      boxExtents=<40 36 15>
      constraintSetup {angularTargetAxis=AxisZ, twistAxis=AxisZ, angularLimitsMax=<0 0 5>, angularLimitsMin=<0 -20 -5>, angularTarget=<0 0 1>}
      localJointOffset=<0 0 3.5>
     }
     {
      boundBone {boneName=dyn_shorts_bk_02_r}
      boxExtents=<10 10 10>
      constraintSetup {angularLimitsMax=<0 0 0>, angularLimitsMin=<0 0 0>, angularTarget=<0 0 0>}
      localJointOffset=<0 0 0>
     }
    ]
   }
   n17 = Flap Node - Bone: dyn_shorts_bk_02_l!(ComponentPose:Component Space Pose Link Structure=@n16, Alpha:float=1) -> (Pose:Component Space Pose Link Structure) {
    is=title FortAnimGraphNode_Flap
    targetComponent=RotationY
    alphaBoolBlend {blendOption=Linear}
    sourceDrives [{
      bUseRange=true
      drivingCurve=null
      multiplier=1
      rangeMax=-0.53
      rangeMin=-2.094
      remappedMax=0
      remappedMin=-1.919
      sourceComponent=RotationZ
      useQuaternion=true
      sourceBone {boneName=calf_l}
     }]
    targetBone {boneName=dyn_shorts_bk_02_l}
   }
   n18 = Flap Node - Bone: dyn_shorts_bk_02_l(ComponentPose=@n17, Alpha=1) {
    targetComponent=ScaleZ
    alphaBoolBlend {blendOption=Linear}
    sourceDrives [{
      bUseRange=true
      drivingCurve=null
      multiplier=1
      rangeMax=-0.53
      rangeMin=-2.094
      remappedMax=0
      remappedMin=-0.8
      sourceComponent=RotationZ
      useQuaternion=true
      sourceBone {boneName=calf_l}
     }]
    targetBone {boneName=dyn_shorts_bk_02_l}
   }
   n19 = Flap Node - Bone: dyn_shorts_bk_02_r!(ComponentPose:Component Space Pose Link Structure=@n18, Alpha:float=1) -> (Pose:Component Space Pose Link Structure) {
    is=title FortAnimGraphNode_Flap
    targetComponent=RotationY
    alphaBoolBlend {blendOption=Linear}
    sourceDrives [{
      bUseRange=true
      drivingCurve=null
      multiplier=1
      rangeMax=-0.53
      rangeMin=-2.094
      remappedMax=0
      remappedMin=-1.99
      sourceComponent=RotationZ
      useQuaternion=true
      sourceBone {boneName=calf_r}
     }]
    targetBone {boneName=dyn_shorts_bk_02_r}
   }
   n20 = Flap Node - Bone: dyn_shorts_bk_02_r(ComponentPose=@n19, Alpha=1) {
    targetComponent=ScaleZ
    alphaBoolBlend {blendOption=Linear}
    sourceDrives [{
      bUseRange=true
      drivingCurve=null
      multiplier=1
      rangeMax=-0.53
      rangeMin=-2.094
      remappedMax=0
      remappedMin=-0.9
      sourceComponent=RotationZ
      useQuaternion=true
      sourceBone {boneName=calf_r}
     }]
    targetBone {boneName=dyn_shorts_bk_02_r}
   }
   n21 = Anim Dynamics - Bone: dyn_bball_necklace_01!(ComponentPose=@n20, GravityOverride:Vector="0,0,0", ExternalForce="0,0,0") -> (Pose) {
    alpha=0.6
    angularDampingOverride=0.99
    angularSpringConstant=1500
    bAngularSpring=true
    bGravityOverrideInSimSpace=true
    bOverrideAngularDamping=true
    bUseGravityOverride=true
    numSolverIterationsPostUpdate=4
    numSolverIterationsPreUpdate=16
    simulationSpace=RootRelative
    alphaBoolBlend {blendOption=Linear}
    boundBone {boneName=dyn_bball_necklace_01}
    physicsBodyDefinitions [{
      sphereCollisionRadius=0
      boundBone {boneName=dyn_bball_necklace_01}
      boxExtents=<20 5 10>
      constraintSetup {angularTargetAxis=AxisY, angularLimitsMax=<0 60 60>, angularLimitsMin=<0 -60 0>, angularTarget=<0 1 0>}
      localJointOffset=<2 0 0>
     }]
   }
   Save cached pose 'cache_01'(Pose=Component To Local(ComponentPose=@n21)) {cacheName=cache_01}
   Sequence Player 'leg_bend_anim'() {playRate=2, sequence=/Game/Characters/Player/Male/Medium/Bodies/M_MED_StreetBasketball/Animation/leg_bend_anim}
   Sequence Player 'M_Jog_NoWep_N'() {sequence=/Game/Characters/Player/Male/Medium/Bodies/M_MED_StreetBasketball/Animation/M_Jog_NoWep_N}
   Sequence Player 'Crouch_C4_Pose'() {sequence=/Game/Characters/Player/Male/Medium/Bodies/M_MED_StreetBasketball/Animation/Crouch_C4_Pose}
   Flap Node - Bone: dyn_bball_necklace_02!(ComponentPose:Component Space Pose Link Structure, Alpha:float=1) -> (Pose:Component Space Pose Link Structure) {
    is=title FortAnimGraphNode_Flap
    targetComponent=RotationZ
    alphaBoolBlend {blendOption=Linear}
    sourceDrives [{
      bUseRange=false
      drivingCurve=null
      multiplier=-2.5
      rangeMax=0
      rangeMin=0
      remappedMax=0
      remappedMin=0
      sourceComponent=RotationZ
      useQuaternion=true
      sourceBone {boneName=spine_04}
     }]
    targetBone {boneName=dyn_bball_necklace_02}
   }
  ]}
 EventGraph {nodes [
   Event BlueprintUpdateAnimation()
   > n1 = Gravity Override State Function DEPRECATED(CharacterPartAnimInstance=Self-Reference(), GameTimeInSeconds=GetGameTimeInSeconds(), GravityOverrideParams=Get Gravity Override Params())
   > Set GravityOverride(GravityOverride=@n1)
   Event BlueprintInitializeAnimation()
   > Set MeshToCopy(MeshToCopy:Skeletal Mesh Component Object Reference=GetOwnerSkeletalMeshComponent()) -> (Output_Get:Skeletal Mesh Component Object Reference)
   TryGetPawnOwner()
  ]}
]
variables [
 Gravity Override Params {type=GravityOverrideParamsStruct}
 GravityOverride {type=Vector}
 MeshToCopy {type=SkeletalMeshComponent}
]
